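# Sound cue for the default impact of the Football pickaxe cosmetic in Battle Royale.
# A Random node picks one of five equally weighted generic impact waves and feeds it through a Modulator at the root.
# Shared pickaxe attenuation, property sheet, melee sound class and player melee concurrency set govern playback, and the cue primes on load.
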SoundCue /BRCosmetics/Sounds/PickAxes/Football/PickAxe_Impact_Default_Football_Cue
root=Modulator_0
nodes [
 {id=Modulator_0, children [Random_1]}
 {id=Random_1, children [WavePlayer_5,WavePlayer_6,WavePlayer_7,WavePlayer_8,WavePlayer_9], properties {weights [1,1,1,1,1]}}
 {id=WavePlayer_5, properties {soundWaveAssetPtr=&/PA_Football_Gen_Impact_05}}
 {id=WavePlayer_6, properties {soundWaveAssetPtr=&/PA_Football_Gen_Impact_04}}
 {id=WavePlayer_7, properties {soundWaveAssetPtr=&/PA_Football_Gen_Impact_03}}
 {id=WavePlayer_8, properties {soundWaveAssetPtr=&/PA_Football_Gen_Impact_02}}
 {id=WavePlayer_9, properties {soundWaveAssetPtr=&/PA_Football_Gen_Impact_01}}
]
properties {
 attenuationSettings=/Game/Sounds/Attenuation/Fort_Pickaxe_Impact_Att
 audioPropertiesSheet=/BRCosmetics/Sounds/Settings/PropertySheets/Pickaxe/APS_Cosmetic_Pickaxe_Impact
 bPrimeOnLoad=true
 maxDistance=4300
 soundClassObject=&Fort_WeaponMelee_PSM
 volumeMultiplier=0.5
 concurrencyOverrides {bLimitToOwner=true, maxCount=2, resolutionRule=StopOldest}
 concurrencySet [&Player_Melee_Weapon_Conc]
}
